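# Physics asset for the AgarBelly pickaxe that sets up a whip-like chain on a two-handed melee handle, alongside proxy bodies for a character.
# Fourteen whip_skin capsules link in sequence from the handle with swing limits and locked twist, ending in a much larger tip capsule.
# Pelvis, lower spine, thigh and calf capsules form a separate body hierarchy with free-rotating joints and no constraint to the whip chain.
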PhysicsAsset /BRCosmetics/Weapons/FORT_Melee/Pickaxe_AgarBelly/Meshes/AgarBelly_Axe_Physics_FE
bodies {
 Melee_TwoHanded_Handle [{shapeType=Capsule, length=20, radius=4, shapeName=Melee_TwoHanded_Handle_capsule, center=<0 0 -15>, rotation=<rot 0 -90 0>}]
 body_calf_l [{shapeType=Capsule, length=30, radius=7, shapeName=body_calf_l_capsule, center=<-20 0 0>, rotation=<rot 90 0 0>}]
 body_calf_r [{shapeType=Capsule, length=30, radius=7, shapeName=body_calf_r_capsule, center=<20 0 0>, rotation=<rot 90 0 0>}]
 body_pelvis [{shapeType=Capsule, length=10, radius=10, shapeName=body_pelvis_capsule}]
 body_spine_01 [{shapeType=Capsule, length=10, radius=12, shapeName=body_spine_01_capsule, center=<5 0 0>}]
 body_spine_02 [{shapeType=Capsule, length=10, radius=12, shapeName=body_spine_02_capsule, center=<15 0 0>}]
 body_thigh_l [{shapeType=Capsule, length=35, radius=7, shapeName=body_thigh_l_capsule, center=<-20 0 0>, rotation=<rot 90 0 0>}]
 body_thigh_r [{shapeType=Capsule, length=35, radius=7, shapeName=body_thigh_r_capsule, center=<20 0 0>, rotation=<rot 90 0 0>}]
 whip_skin_01 [{shapeType=Capsule, length=2, radius=2, center=<0.05 0 0>, rotation=<rot 90 0 0>}]
 whip_skin_02 [{shapeType=Capsule, length=2, radius=2, center=<0.05 0 0>, rotation=<rot 90 0 0>}]
 whip_skin_03 [{shapeType=Capsule, length=2, radius=2, center=<0.05 0 0>, rotation=<rot 90 0 0>}]
 whip_skin_04 [{shapeType=Capsule, length=2, radius=2, center=<0.1 0 0>, rotation=<rot 90 0 0>}]
 whip_skin_05 [{shapeType=Capsule, length=2, radius=2, center=<0.1 0 0>, rotation=<rot 90 0 0>}]
 whip_skin_06 [{shapeType=Capsule, length=2, radius=2, center=<0.1 0 0>, rotation=<rot 90 0 0>}]
 whip_skin_07 [{shapeType=Capsule, length=2, radius=2, center=<0.1 0 0>, rotation=<rot 90 0 0>}]
 whip_skin_08 [{shapeType=Capsule, length=2, radius=2, center=<0.1 0 0>, rotation=<rot 90 0 0>}]
 whip_skin_09 [{shapeType=Capsule, length=2, radius=2, center=<0.1 0 0>, rotation=<rot 90 0 0>}]
 whip_skin_10 [{shapeType=Capsule, length=2, radius=2, center=<0.1 0 0>, rotation=<rot 90 0 0>}]
 whip_skin_11 [{shapeType=Capsule, length=2, radius=2, center=<0.1 0 0>, rotation=<rot 90 0 0>}]
 whip_skin_12 [{shapeType=Capsule, length=2, radius=2, center=<0.1 0 0>, rotation=<rot 90 0 0>}]
 whip_skin_13 [{shapeType=Capsule, length=2, radius=2, center=<0.1 0 0>, rotation=<rot 90 0 0>}]
 whip_skin_14 [{shapeType=Capsule, length=0, radius=11, center=<14 0 0>, rotation=<rot 90 0 0>}]
}
constraints [
 {
  bone1Name=whip_skin_01
  bone2Name=Melee_TwoHanded_Handle
  swing1LimitDegrees=65
  swing2LimitDegrees=65
  twistLimitDegrees=1
  twistMotion=Locked
 }
 {bone1Name=whip_skin_02, bone2Name=whip_skin_01, swing1LimitDegrees=65, swing2LimitDegrees=65, twistLimitDegrees=1, twistMotion=Locked}
 {bone1Name=whip_skin_03, bone2Name=whip_skin_02, swing1LimitDegrees=65, swing2LimitDegrees=65, twistLimitDegrees=1, twistMotion=Locked}
 {bone1Name=whip_skin_04, bone2Name=whip_skin_03, swing1LimitDegrees=65, swing2LimitDegrees=65, twistLimitDegrees=1, twistMotion=Locked}
 {bone1Name=whip_skin_05, bone2Name=whip_skin_04, swing1LimitDegrees=65, swing2LimitDegrees=65, twistLimitDegrees=1, twistMotion=Locked}
 {bone1Name=whip_skin_06, bone2Name=whip_skin_05, swing1LimitDegrees=65, swing2LimitDegrees=65, twistLimitDegrees=1, twistMotion=Locked}
 {bone1Name=whip_skin_07, bone2Name=whip_skin_06, swing1LimitDegrees=65, swing2LimitDegrees=65, twistLimitDegrees=1, twistMotion=Locked}
 {bone1Name=whip_skin_08, bone2Name=whip_skin_07, swing1LimitDegrees=65, swing2LimitDegrees=65, twistLimitDegrees=1, twistMotion=Locked}
 {bone1Name=whip_skin_09, bone2Name=whip_skin_08, swing1LimitDegrees=65, swing2LimitDegrees=65, twistLimitDegrees=1, twistMotion=Locked}
 {bone1Name=whip_skin_10, bone2Name=whip_skin_09, swing1LimitDegrees=65, swing2LimitDegrees=65, twistLimitDegrees=1, twistMotion=Locked}
 {bone1Name=whip_skin_11, bone2Name=whip_skin_10, swing1LimitDegrees=65, swing2LimitDegrees=65, twistLimitDegrees=1, twistMotion=Locked}
 {bone1Name=whip_skin_12, bone2Name=whip_skin_11, swing1LimitDegrees=65, swing2LimitDegrees=65, twistLimitDegrees=1, twistMotion=Locked}
 {bone1Name=whip_skin_13, bone2Name=whip_skin_12, swing1LimitDegrees=65, swing2LimitDegrees=65, twistLimitDegrees=1, twistMotion=Locked}
 {bone1Name=whip_skin_14, bone2Name=whip_skin_13, swing1LimitDegrees=65, swing2LimitDegrees=65, twistLimitDegrees=1, twistMotion=Locked}
 {
  bone1Name=body_spine_01
  bone2Name=body_pelvis
  swing1LimitDegrees=45
  swing1Motion=Free
  swing2LimitDegrees=45
  swing2Motion=Free
  twistLimitDegrees=45
  twistMotion=Free
 }
 {
  bone1Name=body_spine_02
  bone2Name=body_spine_01
  swing1LimitDegrees=45
  swing1Motion=Free
  swing2LimitDegrees=45
  swing2Motion=Free
  twistLimitDegrees=45
  twistMotion=Free
 }
 {
  bone1Name=body_thigh_l
  bone2Name=body_pelvis
  swing1LimitDegrees=45
  swing1Motion=Free
  swing2LimitDegrees=45
  swing2Motion=Free
  twistLimitDegrees=45
  twistMotion=Free
 }
 {
  bone1Name=body_calf_l
  bone2Name=body_thigh_l
  swing1LimitDegrees=45
  swing1Motion=Free
  swing2LimitDegrees=45
  swing2Motion=Free
  twistLimitDegrees=45
  twistMotion=Free
 }
 {
  bone1Name=body_thigh_r
  bone2Name=body_pelvis
  swing1LimitDegrees=45
  swing1Motion=Free
  swing2LimitDegrees=45
  swing2Motion=Free
  twistLimitDegrees=45
  twistMotion=Free
 }
 {
  bone1Name=body_calf_r
  bone2Name=body_thigh_r
  swing1LimitDegrees=45
  swing1Motion=Free
  swing2LimitDegrees=45
  swing2Motion=Free
  twistLimitDegrees=45
  twistMotion=Free
 }
]
